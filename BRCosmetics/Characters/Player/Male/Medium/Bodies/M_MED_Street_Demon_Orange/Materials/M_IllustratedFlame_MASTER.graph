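# Master character material that paints animated, illustrated flames onto the Street Demon Orange skin, layered between the shared character FX functions.
# A static switch picks world-aligned projection on pre-skinned position or panned UV1; both add noise turbulence and smoothstep-thresholded edges, softened by camera distance when projected.
# A mask texture supplies flame, region and shoe channels, with per-region speed, tiling, ramp and gradient settings.
# Curve-atlas color feeds switchable base color and emissive blends, alongside a specular multiplier.
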Material /BRCosmetics/Characters/Player/Male/Medium/Bodies/M_MED_Street_Demon_Orange/Materials/M_IllustratedFlame_MASTER
graph {
 properties {
  bUseMaterialAttributes=true
  bUsedWithMeshDeformer=true
  bUsedWithMorphTargets=true
  bUsedWithSkeletalMesh=true
  pixelDepthOffsetMode=PDOM_Legacy
  refractionMethod=RM_IndexOfRefraction
  shadingModel=MSM_FromMaterialExpression
  subsurfaceProfile=&SS_HeroSkin_02
  usedShadingModels=Default Lit | Subsurface | Clear Coat | Subsurface Profile | Cloth | Eye
 }
 nodes [
  Constant() {r=0.9}
  Constant() {r=0.9}
  n1 = MaterialFunctionCall() {materialFunction=&BaseFN_CharacterFunction_PreFX}
  n2 = GetMaterialAttributes(None=MaterialFunctionCall(
    PreskinnedPosition=@n1.Pre skinned Local Position
    Gmap=@n1.Pre skinned Local Position
    MaterialAttributes=@n1.Pre skinned Local Position
    AO=@n1.Pre skinned Local Position
   ) {materialFunction=&CharacterSkinFX}.Result) {attributeGetTypes [#1, #2, #3]}
  n3 = CurveAtlasRowParameter(CurveTime=NamedRerouteUsage()) {
   atlas=&AthenaCurveAtlas01
   curve=/Game/Characters/Player/Male/Medium/Bodies/M_MED_Street_Demon_Orange/Textures/Curve_StreetDemon_Orange_Flames
   defaultValue=280
   group=Graffiti Flames
   parameterName=Flame Color Curve
  }
  n4 = Multiply(A=@n3, B=ScalarParameter() {defaultValue=1, group=Graffiti Flames, parameterName=Flame Emissive Intensity})
  n5 = NamedRerouteUsage()
  n6 = LinearInterpolate(A=@n2.Base Color, B=@n4, Alpha=@n5) {constB=0}
  n7 = LinearInterpolate(A=@n2.Emissive Color, B=@n4, Alpha=@n5) {constB=0}
  n8 = LinearInterpolate(A=@n2.Specular, B=Multiply(A=@n2.Specular, B=ScalarParameter() {parameterName=Flame Specular Multiplier}), Alpha=Saturate(None=@n5))
  n9 = MaterialFunctionCall(
   AO=@n1.AO
   Distance Blend=@n1.AO
   Fresnel=@n1.AO
   Gmap=@n1.AO
   Pre Skinned Local Position=@n1.AO
   Material=SetMaterialAttributes!(
    MaterialAttributes=@n2.MaterialAttributes
    Base Color=StaticSwitchParameter(True=@n6, False=@n2.Base Color) {defaultValue=true, group=Graffiti Flames, parameterName=FlameBaseColor}
    Emissive Color=StaticSwitchParameter(True=@n7, False=@n2.Emissive Color) {group=Graffiti Flames, parameterName=FlameEmission}
    Specular=StaticSwitchParameter(True=@n8, False=@n2.Specular) {defaultValue=true, group=Graffiti Flames, parameterName=FlameSpecularMultiplier}
   ) {attributeSetTypes [#1, #2, #3]}
  ) {materialFunction=&BaseFN_CharacterFunction_PostFX}
  MaterialFunctionCall(In=@n9.Result) {materialFunction=&CINE_MF_CharPostFX}
  n10 = TextureSampleParameter2D() {
   group=Graffiti Flames
   parameterName=Flame Masks
   samplerType=SAMPLERTYPE_Masks
   sortPriority=5
   texture=/BRCosmetics/Characters/Player/Female/Medium/Bodies/F_MED_Graffiti_01/Skins/Future_Orange/Textures/T_F_MED_FutureOrange_Body_FlameMasks
  }
  NamedRerouteDeclaration(None=@n10.R) {name=Flame Mask, nodeColor=<rgba 0.8516 1 0 1>}
  n11 = TextureObjectParameter() {
   group=Graffiti Flames
   parameterName=Flame Texture
   samplerType=SAMPLERTYPE_LinearColor
   texture=/BRCosmetics/Characters/Player/Male/Medium/Bodies/M_MED_Street_Demon_Orange/Textures/T_MED_StreetDemonOrange_Body_Flames
  }
  n12 = NamedRerouteUsage()
  n13 = ComponentMask(None=@n12) {r=true}
  n14 = ComponentMask(None=@n12) {g=true}
  n15 = TextureObjectParameter() {
   group=Graffiti Flames
   parameterName=Turbulence Noise Texture
   samplerType=SAMPLERTYPE_Grayscale
   texture=&LowResBlurredNoise
  }
  n16 = Constant3Vector() {constant=<rgba 64 64 64 0>}
  n17 = ComponentMask(None=NamedRerouteUsage()) {r=true}
  n18 = Time()
  n19 = Add(
   A=MaterialFunctionCall(Z=Multiply(A=@n18) {constB=-2}) {materialFunction=&MakeFloat3}.Result
   B=NamedRerouteUsage()
  )
  n20 = ComponentMask(None=NamedRerouteUsage()) {r=true}
  n21 = Multiply(A=Saturate(None=Sign(None=@n20))) {constB=0.5}
  n22 = VertexInterpolator(VS=PreSkinnedNormal())
  n23 = Saturate(None=Sign(None=ComponentMask(None=@n22.PS) {r=true}))
  n24 = LinearInterpolate(A=Constant3Vector() {constant=<rgba 32 -16 16 0>}, B=Constant3Vector() {constant=<rgba 32 16 16 0>}, Alpha=Divide(A=Add(A=@n23, B=@n21)))
  n25 = Multiply(
   A=MaterialFunctionCall(TextureObject=@n15, TextureSize=Multiply(A=@n16, B=@n17), WorldPosition=Add(A=@n19, B=@n24), World Space Normal=@n22.PS) {materialFunction=&WorldAlignedTexture}.XY Texture
   B=ScalarParameter() {defaultValue=4, group=Graffiti Flames, parameterName=Turbulence Intensity}
  ) {constB=10}
  n26 = Add(A=NamedRerouteUsage(), B=@n24) {constB=0}
  n27 = Multiply(A=Multiply(A=Time()) {constB=-10}, B=NamedRerouteUsage())
  n28 = Constant3Vector() {constant=<rgba 1 1 0 0>}
  n29 = MaterialFunctionCall(
   TextureObject=@n11
   TextureSize=Multiply(
    A=Constant3Vector() {constant=<rgba 128 128 128 0>}
    B=MaterialFunctionCall(X=@n13, Y=@n13, Z=@n14) {materialFunction=&MakeFloat3}.Result
   )
   WorldPosition=Add(
    A=Add(A=@n25, B=@n26)
    B=MaterialFunctionCall(Z=@n27) {materialFunction=&MakeFloat3}.Result
   )
   World Space Normal=Normalize(VectorInput=Multiply(A=@n22.PS, B=@n28))
  ) {materialFunction=&WorldAlignedTexture}
  n30 = Add(
   A=ScalarParameter() {defaultValue=1, group=Graffiti Flames, parameterName=Flame Threshold}
   B=Multiply(A=ComponentMask(None=@n29.XY Texture) {g=true}) {constB=0.1}
  ) {constB=0}
  n31 = Subtract(A=Distance(A=WorldPosition().XYZ, B=CameraPositionWS())) {constB=150}
  n32 = LinearInterpolate(
   A=ScalarParameter() {defaultValue=0.02, group=Graffiti Flames, parameterName=FlameEdgeBlur_Close, sortPriority=40}
   B=ScalarParameter() {defaultValue=0.06, group=Graffiti Flames, parameterName=FlameEdgeBlur_Far, sortPriority=41}
   Alpha=Saturate(None=Divide(A=@n31) {constB=500})
  )
  n33 = NamedRerouteUsage()
  n34 = ComponentMask(None=@n29.XY Texture) {r=true}
  n35 = SmoothStep(Min=Subtract(A=@n30, B=@n32), Max=Add(A=@n30, B=@n32), Value=Add(A=@n33, B=@n34))
  n36 = Multiply(A=TextureCoordinate() {coordinateIndex=1}, B=NamedRerouteUsage())
  n37 = TextureCoordinate() {coordinateIndex=1, uTiling=2, vTiling=2}
  n38 = NamedRerouteUsage()
  n39 = Time()
  n40 = NamedRerouteUsage()
  n41 = TextureObjectParameter() {
   group=Graffiti Flames
   parameterName=Turbulence Noise Texture
   samplerType=SAMPLERTYPE_Grayscale
   texture=&LowResBlurredNoise
  }
  n42 = TextureSample(UVs=Panner(Coordinate=Multiply(A=@n37, B=@n38), Time=Multiply(A=@n39, B=@n40)) {speedY=0.02}, Tex=@n41) {
   samplerType=SAMPLERTYPE_LinearColor
   texture=&T_MED_StreetDemonOrange_Body_Flames
  }
  n43 = ScalarParameter() {defaultValue=4, group=Graffiti Flames, parameterName=Turbulence Intensity}
  n44 = MaterialFunctionCall(X=Multiply(A=@n42.R, B=Multiply(B=@n43) {constA=0.01})) {materialFunction=&MakeFloat2}
  n45 = Time()
  n46 = NamedRerouteUsage()
  n47 = TextureObjectParameter() {
   group=Graffiti Flames
   parameterName=Flame Texture
   samplerType=SAMPLERTYPE_LinearColor
   texture=&T_MED_StreetDemonOrange_Body_Flames
  }
  n48 = TextureSample(UVs=Panner(Coordinate=Add(A=@n36, B=@n44.Result), Time=Multiply(A=@n45, B=@n46)) {speedY=0.1}, Tex=@n47) {
   samplerType=SAMPLERTYPE_LinearColor
   texture=&T_MED_StreetDemonOrange_Body_Flames
  }
  n49 = Add(A=ScalarParameter() {defaultValue=1, group=Graffiti Flames, parameterName=Flame Threshold}, B=Multiply(A=@n48.G) {constB=0.1}) {constB=0}
  n50 = ScalarParameter() {defaultValue=0.02, group=Graffiti Flames, parameterName=FlameEdgeBlur, sortPriority=40}
  n51 = NamedRerouteUsage()
  n52 = SmoothStep(Min=Subtract(A=@n49, B=@n50), Max=Add(A=@n49, B=@n50), Value=Add(A=@n51, B=@n48.R))
  NamedRerouteDeclaration(None=Multiply(A=StaticSwitchParameter(True=@n35, False=@n52) {parameterName=Use World Aligned Textures}, B=NamedRerouteUsage())) {name=FlameMask, nodeColor=<rgba 0 0.625 1 1>}
  n53 = Add(A=@n33, B=Multiply(A=@n34) {constB=0.2})
  n54 = NamedRerouteUsage()
  n55 = Add(A=@n51, B=Multiply(A=@n48.R) {constB=0.2})
  n56 = NamedRerouteUsage()
  NamedRerouteDeclaration(None=StaticSwitchParameter(True=Power(Base=@n53, Exp=@n54), False=Power(Base=@n55, Exp=@n56)) {parameterName=Use World Aligned Textures}) {name=FlameGradient, nodeColor=<rgba 0.8516 1 0 1>}
  NamedRerouteDeclaration(None=MaterialFunctionCall!(
    1=ScalarParameter() {defaultValue=1, group=Graffiti Flames, parameterName=Flame Speed R1, sortPriority=500}
    2=ScalarParameter() {defaultValue=0.2, group=Graffiti Flames, parameterName=Flame Speed R2, sortPriority=501}
    3=ScalarParameter() {defaultValue=1, group=Graffiti Flames, parameterName=Flame Speed R3, sortPriority=502}
    4
    A=NamedRerouteUsage()
   ) {materialFunction=/Engine/Functions/Engine_MaterialFunctions02/Lerp_Multiple_Float}.Lerp 3 Inputs) {name=Flame Speed, nodeColor=<rgba 0.8516 1 0 1>}
  NamedRerouteDeclaration(None=@n10.G) {name=Region Mask, nodeColor=<rgba 0.5469 0 1 1>}
  n57 = ScalarParameter() {defaultValue=0.5, group=Graffiti Flames, parameterName=Tile X R1, sortPriority=600}
  n58 = ScalarParameter() {defaultValue=1, group=Graffiti Flames, parameterName=Tile Y R1, sortPriority=601}
  n59 = ScalarParameter() {defaultValue=1, group=Graffiti Flames, parameterName=Tile X R2, sortPriority=602}
  n60 = ScalarParameter() {defaultValue=1, group=Graffiti Flames, parameterName=Tile Y R2, sortPriority=603}
  n61 = ScalarParameter() {defaultValue=1, group=Graffiti Flames, parameterName=Tile X R3, sortPriority=604}
  n62 = ScalarParameter() {defaultValue=1, group=Graffiti Flames, parameterName=Tile Y R3, sortPriority=605}
  NamedRerouteDeclaration(None=MaterialFunctionCall!(
    1=MaterialFunctionCall(X=@n57, Y=@n58) {materialFunction=&MakeFloat2}.Result
    2=MaterialFunctionCall(X=@n59, Y=@n60) {materialFunction=&MakeFloat2}.Result
    3=MaterialFunctionCall(X=@n61, Y=@n62) {materialFunction=&MakeFloat2}.Result
    4
    A=NamedRerouteUsage()
   ) {materialFunction=/Engine/Functions/Engine_MaterialFunctions02/Lerp_Multiple_Float2}.Lerp 3 Inputs) {name=Flame Tiling, nodeColor=<rgba 1 0 0.9844 1>}
  n63 = ComponentMask(None=TextureCoordinate() {coordinateIndex=1, uTiling=2}) {g=true}
  n64 = MaterialFunctionCall(
   1=ScalarParameter() {defaultValue=0.8, group=Graffiti Flames, parameterName=Ramp Low R1, sortPriority=1000}
   2=ScalarParameter() {defaultValue=0.45, group=Graffiti Flames, parameterName=Ramp Low R2, sortPriority=1002}
   3=LinearInterpolate(
    A=ScalarParameter() {defaultValue=0.5, group=Graffiti Flames, parameterName=Ramp Low R3, sortPriority=1004}
    B=ScalarParameter() {defaultValue=0.5, group=Graffiti Flames, parameterName=Ramp Low R3_2, sortPriority=1005}
    Alpha=NamedRerouteUsage()
   )
   A=NamedRerouteUsage()
  ) {materialFunction=&Lerp_Multiple_Float}
  n65 = Constant() {r=1}
  n66 = Constant()
  n67 = MaterialFunctionCall(
   1=ScalarParameter() {defaultValue=1, group=Graffiti Flames, parameterName=Ramp High R1, sortPriority=1001}
   2=ScalarParameter() {defaultValue=1, group=Graffiti Flames, parameterName=Ramp High R2, sortPriority=1003}
   3=ScalarParameter() {defaultValue=1, group=Graffiti Flames, parameterName=Ramp High R3, sortPriority=1007}
   A=NamedRerouteUsage()
  ) {materialFunction=&Lerp_Multiple_Float}
  NamedRerouteDeclaration(None=Saturate(None=MaterialFunctionCall(Input=@n63, Input Low=@n64.Lerp 3 Inputs, Input High=@n65, Target Low=@n66, Target High=@n67.Lerp 3 Inputs) {materialFunction=&RemapValueRange}.Result)) {name=Flame Ramp, nodeColor=<rgba 0 1 0.0625 1>}
  NamedRerouteDeclaration(None=MaterialFunctionCall(
    1=ScalarParameter() {defaultValue=1, group=Graffiti Flames, parameterName=Flame Color Gradient Exponent R1, sortPriority=71}
    2=ScalarParameter() {defaultValue=1, group=Graffiti Flames, parameterName=Flame Color Gradient Exponent R2, sortPriority=72}
    3=ScalarParameter() {defaultValue=1, group=Graffiti Flames, parameterName=Flame Color Gradient Exponent R3, sortPriority=73}
    A=NamedRerouteUsage()
   ) {materialFunction=&Lerp_Multiple_Float}.Lerp 3 Inputs) {name=Flame Color Gradient Exponent, nodeColor=<rgba 0.4062 0 1 1>}
  NamedRerouteDeclaration(None=@n1.Pre skinned Local Position) {name=Preskinned Location, nodeColor=<rgba 1 0 0.1159 1>}
  NamedRerouteDeclaration(None=@n10.B) {name=Shoes Gradient, nodeColor=<rgba 0.1953 0 1 1>}
  Transform(None=VertexNormalWS()) {transformSourceType=TRANSFORMSOURCE_World, transformType=TRANSFORM_Local}
 ]
}
