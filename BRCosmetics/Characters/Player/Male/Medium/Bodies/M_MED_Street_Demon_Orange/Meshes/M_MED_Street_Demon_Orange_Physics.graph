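# Physics asset for the Street Demon Orange male medium player body cosmetic, setting up secondary motion for its bandolier, front collar flaps and zipper.
# Each dynamic bone pairs with its MSTR bone through a constraint with angular limits, using capsules for bandolier and collars and boxes for the zipper.
# The bandolier joint locks twist, while the others allow slight twist.
# An unconstrained right thigh capsule is the only non-accessory body, likely a collision blocker.
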PhysicsAsset /BRCosmetics/Characters/Player/Male/Medium/Bodies/M_MED_Street_Demon_Orange/Meshes/M_MED_Street_Demon_Orange_Physics
bodies {
 dyn_bandolier1 [{shapeType=Capsule, length=0.1801, radius=1.0861, center=<1.3098 0 0>, rotation=<rot -0.0001 0 0>}]
 dyn_bandolier_MSTR [{shapeType=Capsule, length=0.75, radius=1.5, rotation=<rot 90 0 0>}]
 dyn_collar_front_l_01 [{shapeType=Capsule, length=0.5099, radius=2.1795, center=<3.3892 -0.0001 0>}]
 dyn_collar_front_l_MSTR [{shapeType=Capsule, length=0.505, radius=0.505, rotation=<rot 0 -90 0>}]
 dyn_collar_front_r_01 [{shapeType=Capsule, length=0.8126, radius=2.2833, center=<3.3892 0 0>}]
 dyn_collar_front_r_MSTR [{shapeType=Capsule, length=0.505, radius=0.505, rotation=<rot 0 -90 0>}]
 dyn_zipper [{shapeType=Box, extentX=2.01, extentY=1.01, extentZ=3.51, center=<2.6655 0.1333 0>, rotation=<rot 90 0 0>}]
 dyn_zipper_MSTR [{shapeType=Box, extentX=1.6785, extentY=2.3681, extentZ=3.1272, center=<0.2933 0.7373 -0.0236>, rotation=<rot 0.6698 148.8709 -86.1987>}]
 thigh_r [{shapeType=Capsule, length=6.2831, radius=11.0199, center=<10.2627 0.0004 0>, rotation=<rot 90 0.0198 0.0198>}]
}
constraints [
 {bone1Name=dyn_zipper, bone2Name=dyn_zipper_MSTR, swing1LimitDegrees=45, swing2LimitDegrees=30, twistLimitDegrees=2}
 {
  bone1Name=dyn_bandolier_MSTR
  bone2Name=dyn_bandolier1
  swing1LimitDegrees=2.5
  swing2LimitDegrees=20
  twistLimitDegrees=0
  twistMotion=Locked
 }
 {bone1Name=dyn_collar_front_l_01, bone2Name=dyn_collar_front_l_MSTR, swing1LimitDegrees=10, swing2LimitDegrees=10, twistLimitDegrees=2.5}
 {bone1Name=dyn_collar_front_r_01, bone2Name=dyn_collar_front_r_MSTR, swing1LimitDegrees=10, swing2LimitDegrees=10, twistLimitDegrees=2.5}
]
