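# Post-copy AnimBlueprint for the Street Demon Orange body cosmetic that copies the attached parent's pose and adds secondary motion to collar, zipper and bandolier.
# Each update, head and right upper-arm yaw relative to spine_05 become clamped alphas.
# These drive additive bone corrections: two-stage back collar, plus both front collar sides from the right arm.
# Weighted constraints then position zipper and bandolier masters before a root-space RigidBody simulation using an override physics asset.
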AnimBlueprint /BRCosmetics/Characters/Player/Male/Medium/Bodies/M_MED_Street_Demon_Orange/Meshes/M_MED_Street_Demon_Orange_AnimBP
target_skeleton=&/M_MED_Street_Demon_Orange_Skeleton
graphs [
 AnimGraph {nodes [
   n1 = Get head_rot_back() -> (head_rot_back:double)
   n2 = "Transform (Modify) Bone - Bone: dyn_collar_back_l_01"!(ComponentPose=Local To Component(LocalPose=Use cached pose 'collarRotCache'()), Alpha=@n1, Translation="0,-1,0", Rotation="0,0,-20") -> (Pose) {
    rotationMode=BMM_Additive
    rotationSpace=BCS_BoneSpace
    scaleSpace=BCS_BoneSpace
    translationMode=BMM_Additive
    translationSpace=BCS_BoneSpace
    alphaBoolBlend {blendOption=Linear}
    boneToModify {boneName=dyn_collar_back_l_01}
    rotation {roll=-20}
    scale=<2 2 2>
    translation {y=-1}
   }
   n3 = "Transform (Modify) Bone - Bone: dyn_collar_back_r_01"!(ComponentPose=@n2, Alpha=@n1, Translation="2,-2,0", Rotation="0,-20,-45") -> (Pose) {
    rotationMode=BMM_Additive
    rotationSpace=BCS_BoneSpace
    scaleSpace=BCS_BoneSpace
    translationMode=BMM_Additive
    translationSpace=BCS_BoneSpace
    alphaBoolBlend {blendOption=Linear}
    boneToModify {boneName=dyn_collar_back_r_01}
    rotation {roll=-45, yaw=-20}
    scale=<2 2 2>
    translation=<2 -2>
   }
   n4 = Get Head_Rot_FAR_Back() -> (Head_Rot_FAR_Back:double)
   n5 = "Transform (Modify) Bone - Bone: dyn_collar_back_l_01"(ComponentPose=@n3, Alpha=@n4, Translation="4,-6,0", Rotation="-30,0,-40") {
    rotationMode=BMM_Additive
    rotationSpace=BCS_BoneSpace
    scaleSpace=BCS_BoneSpace
    translationMode=BMM_Additive
    translationSpace=BCS_BoneSpace
    alphaBoolBlend {blendOption=Linear}
    boneToModify {boneName=dyn_collar_back_l_01}
    rotation {pitch=-30, roll=-40}
    scale=<2 2 2>
    translation=<4 -6>
   }
   n6 = "Transform (Modify) Bone - Bone: dyn_collar_back_r_01"(ComponentPose=@n5, Alpha=@n4, Translation="2,-7,0", Rotation="-30,-20,0") {
    rotationMode=BMM_Additive
    rotationSpace=BCS_BoneSpace
    scaleSpace=BCS_BoneSpace
    translationMode=BMM_Additive
    translationSpace=BCS_BoneSpace
    alphaBoolBlend {blendOption=Linear}
    boneToModify {boneName=dyn_collar_back_r_01}
    rotation {pitch=-30, yaw=-20}
    scale=<2 2 2>
    translation=<2 -7>
   }
   n7 = "Transform (Modify) Bone - Bone: dyn_collar_front_r_MSTR"(ComponentPose=@n6, Alpha=Get ADS_UpArm_Rot() -> (ADS_UpArm_Rot:double), Translation="-4.2,3,0", Rotation="20,-40,0", Scale="1,1,1") {
    rotationMode=BMM_Additive
    rotationSpace=BCS_BoneSpace
    translationMode=BMM_Additive
    translationSpace=BCS_BoneSpace
    alphaBoolBlend {blendOption=Linear}
    boneToModify {boneName=dyn_collar_front_r_MSTR}
    rotation {pitch=20, yaw=-40}
    translation=<-4.2 3>
   }
   n8 = "Transform (Modify) Bone - Bone: dyn_collar_front_l_MSTR"(ComponentPose=@n7, Alpha=Get ADS_UpArm_Rot_R() -> (ADS_UpArm_Rot_R:double), Translation="-2,2,0", Rotation="0,0,0", Scale="1,1,1") {
    rotationMode=BMM_Additive
    rotationSpace=BCS_BoneSpace
    translationMode=BMM_Additive
    translationSpace=BCS_BoneSpace
    alphaBoolBlend {blendOption=Linear}
    boneToModify {boneName=dyn_collar_front_l_MSTR}
    translation=<-2 2>
   }
   n9 = "Constraint - dyn_zipper_MSTR (PPPP)"!(
    ComponentPose=@n8
    Alpha=1
    ConstraintWeights_0=0.786
    ConstraintWeights_1:float=0.0191
    ConstraintWeights_2:float=0.1486
    ConstraintWeights_3:float=0.0463
   ) -> (Pose) {
    alphaBoolBlend {blendOption=Linear}
    boneToModify {boneName=dyn_zipper_MSTR}
    constraintSetup [
     {targetBone {boneName=pelvis}}
     {targetBone {boneName=thigh_twist_01_l}}
     {targetBone {boneName=thigh_twist_01_r}}
     {targetBone {boneName=spine_01}}
    ]
    constraintWeights [1,1,1,1]
   }
   n10 = "Constraint - dyn_bandolier_MSTR (PPPPPP)"!(
    ComponentPose=@n9
    Alpha=1
    ConstraintWeights_0=0.0074
    ConstraintWeights_1:float=0.3531
    ConstraintWeights_2:float=0.1322
    ConstraintWeights_3:float=0.0679
    ConstraintWeights_4:float=0.0107
    ConstraintWeights_5:float=0.4287
   ) -> (Pose) {
    alphaBoolBlend {blendOption=Linear}
    boneToModify {boneName=dyn_bandolier_MSTR}
    constraintSetup [
     {targetBone {boneName=thigh_twist_01_r}}
     {targetBone {boneName=spine_01}}
     {targetBone {boneName=spine_02}}
     {targetBone {boneName=spine_03}}
     {targetBone {boneName=spine_04}}
     {targetBone {boneName=pelvis}}
    ]
    constraintWeights [1,1,1,1,1,1]
   }
   n11 = RigidBody(ComponentPose=@n10) {
    bOverrideWorldGravity=true
    overridePhysicsAsset=&/M_MED_Street_Demon_Orange_Physics
    simulationSpace=BaseBoneSpace
    alphaBoolBlend {blendOption=Linear}
    baseBoneRef {boneName=root}
    componentAppliedLinearAccClamp=<3000 3000 3000>
    componentLinearAccScale=<0.3 0.3 0.4>
    componentLinearVelScale=<0.15 0.15 0.2>
    overrideWorldGravity {z=-980}
   }
   Output Pose(Result=Component To Local(ComponentPose=@n11)) {layerGroup=None}
   Save cached pose 'collarRotCache'(Pose=Copy Pose From Mesh() {bUseAttachedParent=true}) {cacheName=collarRotCache}
  ]}
 EventGraph {nodes [
   Event BlueprintUpdateAnimation()
   > n1 = Sequence!(execute) -> (then_0, then_1, then_2:Exec, then_3:Exec, then_4:Exec, then_5:Exec, then_6:Exec)
   n2 = GetDeltaTransformFromRefPose(self=GetOwnerSkeletalMeshComponent(), BoneName=head, BaseName=spine_05)
   > Set headSpineRot(headSpineRot:double=BreakRotator(InRot=BreakTransform(InTransform=@n2).Rotation).Yaw) -> (Output_Get:double)
   Set head_rot_back(execute=@n1.then_1, head_rot_back:double=MapRangeClamped(Value=Get headSpineRot() -> (headSpineRot:double), InRangeA=-30, InRangeB=-60)) -> (Output_Get:double)
   Set Head_Rot_FAR_Back(execute=@n1.then_2, Head_Rot_FAR_Back:double=MapRangeClamped(Value=Get headSpineRot(), InRangeA=-60, InRangeB=-96)) -> (Output_Get:double)
   n3 = GetDeltaTransformFromRefPose(self=GetOwnerSkeletalMeshComponent(), BoneName=upperarm_r, BaseName=spine_05)
   Set UpArm_Spine_Rot(execute=@n1.then_3, UpArm_Spine_Rot:double=BreakRotator(InRot=BreakTransform(InTransform=@n3).Rotation).Yaw) -> (Output_Get:double)
   Set ADS_UpArm_Rot(execute=@n1.then_4, ADS_UpArm_Rot:double=MapRangeClamped(Value=Get UpArm_Spine_Rot() -> (UpArm_Spine_Rot:double), InRangeA=70, InRangeB=140)) -> (Output_Get:double)
   n4 = GetDeltaTransformFromRefPose(self=GetOwnerSkeletalMeshComponent(), BoneName=upperarm_r, BaseName=spine_05)
   Set UpArm_Spine_Rot_R(execute=@n1.then_5, UpArm_Spine_Rot_R:double=BreakRotator(InRot=BreakTransform(InTransform=@n4).Rotation).Yaw) -> (Output_Get:double)
   Set ADS_UpArm_Rot_R(
    execute=@n1.then_6
    ADS_UpArm_Rot_R:double=MapRangeClamped(Value=Get UpArm_Spine_Rot_R() -> (UpArm_Spine_Rot_R:double), InRangeA=70, InRangeB=140)
   ) -> (Output_Get:double)
  ]}
]
variables [
 ADS_FCollar_Rot_Back {type=number}
 ADS_Rot_Forward {type=number}
 ads_rotate_back {type=number}
 ADS_UpArm_Rot {type=number}
 ADS_UpArm_Rot_R {type=number}
 head_rot_back {type=number}
 Head_Rot_FAR_Back {type=number}
 Head_Rot_Forward {type=number}
 headSpineRot {type=number}
 UpArm_Spine_Rot {type=number}
 UpArm_Spine_Rot_R {type=number}
]
