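# Physics asset for the Street Demon Orange medium male player cosmetic, covering the head and what its bone name suggests is a dynamic hat brim.
# It has two bodies: a capsule on head and a thin box on dyn_brim.
# A single constraint joins dyn_brim to head with locked twist and tight, uneven swing limits, so the brim can tilt only slightly relative to the head.
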PhysicsAsset /BRCosmetics/Characters/Player/Male/Medium/Bodies/M_MED_Street_Demon_Orange/Meshes/Parts/M_MED_Street_Demon_Orange_Physics
bodies {
 dyn_brim [{shapeType=Box, extentX=6.01, extentY=1.51, extentZ=14.51, center=<3.0774 0.2375 0>}]
 head [{shapeType=Capsule, length=4.024, radius=7.5001, center=<0 4.3674 0>, rotation=<rot -90 179.9802 -179.9802>}]
}
constraints [{bone1Name=dyn_brim, bone2Name=head, swing1LimitDegrees=10, swing2LimitDegrees=2.5, twistLimitDegrees=0, twistMotion=Locked}]
